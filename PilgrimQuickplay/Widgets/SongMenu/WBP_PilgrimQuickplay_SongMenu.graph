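# Pilgrim Quickplay song menu tab that lets players restart the song or exit to backstage, each behind a confirmation screen.
# Options show only when all players are local; otherwise a solo-only warning appears, rechecked as users join or leave.
# A stateful event listener caches the player's MusicBattleInstance, which gates both buttons.
# Confirmed actions send requests through the Pilgrim controller component, then deactivate the menu and broadcast a quest-screen hide event.
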WidgetBlueprint /PilgrimQuickplay/Widgets/SongMenu/WBP_PilgrimQuickplay_SongMenu
graphs [
 Close Song Menu {nodes [
   Close Song Menu() {is=title K2Node_FunctionEntry}
   > DeactivateWidget()
   > n1 = GetEventRouter!(execute, Scope:Object Class Reference=/Script/FortniteGame.FortPlayspace, ContextActor:Actor Object Reference=GetOwningPlayer()) -> (then, ReturnValue:Gameplay Event Router Component Object Reference)
   > Broadcast Event!(
    execute
    self:Gameplay Event Router Component Object Reference=@n1
    EventData:Show Sparks Quest Screen Event Structure=Make Show Sparks Quest Screen Event!(bShouldShow:bool=False) -> (ShowSparksQuestScreenEvent:Show Sparks Quest Screen Event Structure) {is=title K2Node_MakeStruct}
    EventContext:Object Reference
   ) -> (then, ReturnValue:bool)
  ]}
 ConfirmRestart {nodes [
   ConfirmRestart() {is=title K2Node_FunctionEntry}
   Return Node!(execute, ScreenSwitcher:Common Visibility Switcher Object Reference)
  ]}
 EventGraph {nodes [
   "On Clicked (BackButton)"!() -> (OutputDelegate, then, Button:Common Button Base Object Reference) {is=title K2Node_ComponentBoundEvent}
   n1 = Get ScreenSwitcher() -> (ScreenSwitcher:Common Activatable Widget Switcher Object Reference)
   > n2 = Branch(Condition="Equal (Integer)"(A=GetActiveWidgetIndex!(self:Widget Switcher Object Reference=@n1) -> (ReturnValue:int=0)))
   > Close Song Menu() {is=title K2Node_CallFunction}
   SetActiveWidgetIndex(execute=@n2.else, self=@n1, Index=0)
   "On Clicked (YesRestartButton_Mojo)"!() -> (OutputDelegate, then, Button:Common Button Base Object Reference) {is=title K2Node_ComponentBoundEvent}
   > Restart Song()
   > Close Song Menu()
   "On Clicked (YesLeaveButton_Mojo)"!() -> (OutputDelegate, then, Button:Common Button Base Object Reference) {is=title K2Node_ComponentBoundEvent}
   > Exit to Backstage()
   > Close Song Menu()
   "On Clicked (NoReturnButton_Mojo)"!() -> (OutputDelegate, then, Button:Common Button Base Object Reference) {is=title K2Node_ComponentBoundEvent}
   n3 = Get ScreenSwitcher()
   > SetActiveWidgetIndex(self=@n3, Index=0)
   "On Clicked (NoRestartButton_Mojo)"!() -> (OutputDelegate, then, Button:Common Button Base Object Reference) {is=title K2Node_ComponentBoundEvent}
   > SetActiveWidgetIndex(self=Get ScreenSwitcher(), Index=0)
   n4 = "On Active Widget Index Changed BP (ScreenSwitcher)"!() -> (OutputDelegate, then, ActiveWidget:Widget Object Reference, ActiveIndex:int=0) {is=title K2Node_ComponentBoundEvent}
   > SetFocus(self=Select!(
     Option 0:Widget Object Reference=Get ExitToBackstage_Mojo() -> (ExitToBackstage_Mojo:WBP FM Core Button Regular Object Reference)
     Option 1:Widget Object Reference=Get NoRestartButton_Mojo() -> (NoRestartButton_Mojo:WBP FM Core Button Regular Object Reference)
     Option 2:Widget Object Reference=Get NoReturnButton_Mojo() -> (NoReturnButton_Mojo:WBP FM Core Button Regular Object Reference)
     Index:int=@n4.ActiveIndex
    ) -> (ReturnValue:Widget Object Reference))
   > SetRenderOpacity(
    self=Get BackgroundColor() -> (BackgroundColor:Border Object Reference)
    InOpacity=Select(Option 0:float=0.5, Option 1:float=1, Option 2:float=1, Index=@n4.ActiveIndex) -> (ReturnValue:float)
   )
   "On Clicked (RestartButton_Mojo)"!() -> (OutputDelegate, then, Button:Common Button Base Object Reference) {is=title K2Node_ComponentBoundEvent}
   > SetActiveWidgetIndex(self=Get ScreenSwitcher(), Index=1)
   "On Clicked (ExitToBackstage_Mojo)"!() -> (OutputDelegate, then, Button:Common Button Base Object Reference) {is=title K2Node_ComponentBoundEvent}
   > SetActiveWidgetIndex(self=@n3, Index=2)
   n5 = OnPlayspaceUserChanged!() -> (OutputDelegate, then, AddedUser:"Playspace User Structure (by ref)") {is=title K2Node_CustomEvent}
   > Update All Button Status()
   Event Construct()
   > Update All Button Status()
   Event On Activated()
   > Update All Button Status()
   Event Destruct()
   > Parent: Destruct()
   > n6 = Sequence()
   n7 = Get PilgrimMBIStartedEventHandle() -> (PilgrimMBIStartedEventHandle:Async Action Start Listening To Stateful Event Object Reference)
   > Is Valid(InputObject=@n7)
   > StopListeningToStatefulEvent(self:Async Action Start Listening To Stateful Event Object Reference=@n7)
   > Set PilgrimMBIStartedEventHandle(PilgrimMBIStartedEventHandle:Async Action Start Listening To Stateful Event Object Reference) -> (Output_Get:Async Action Start Listening To Stateful Event Object Reference)
   n8 = Get SparksMusicPlayspace() -> (SparksMusicPlayspace:Sparks Music Playspace Object Reference)
   Is Valid(exec=@n6.then_1, InputObject=@n8)
   n9 = Create Event()
   > Unbind Event from Notify Playspace User Added(self:Playspace Object Reference=@n8, Delegate=@n9) {is=title K2Node_RemoveDelegate}
   > Unbind Event from Notify Playspace User Removed(self:Playspace Object Reference=@n8, Delegate=@n9) {is=title K2Node_RemoveDelegate}
   Event OnInitialized()
   > n10 = PilgrimWaitForEventRouter!(execute:Exec, ContextActor:Actor Object Reference=GetOwningPlayer()) -> (then:Exec, OnEventRouterAvailable:Exec, EventRouter:Gameplay Event Router Component Object Reference) {is=title K2Node_AsyncAction}
   n11 = GetSparksMusicPlayspaceOfClass!(
    PlayspaceType:Sparks Music Playspace Class Reference=/Script/SparksMusicPlayspaceRuntime.SparksMusicPlayspace
    Actor:Actor Object Reference=GetOwningPlayer()
    bMatchChildTypes:bool=false
   ) -> (ReturnValue:Sparks Music Playspace Object Reference)
   > Is Valid(InputObject=@n11)
   > n12 = Set SparksMusicPlayspace(SparksMusicPlayspace:Sparks Music Playspace Object Reference=@n11) -> (Output_Get:Sparks Music Playspace Object Reference)
   > Bind Event to Notify Playspace User Added(self:Playspace Object Reference=@n12, Delegate=@n5) {is=title K2Node_AddDelegate}
   > Bind Event to Notify Playspace User Removed(self:Playspace Object Reference=@n12, Delegate=@n5) {is=title K2Node_AddDelegate}
   n13 = StartListeningToStatefulEvent!(
    execute:Exec=@n10.OnEventRouterAvailable
    Target:Gameplay Event Router Component Object Reference=@n10
    EventType:Script Struct Object Reference=/Script/PilgrimCoreRuntime.PilgrimEvent_MBI_Started
    Context:Object Reference
    EventBubblingRule:EEventBubblingRule Enum
    None:%=@n10
   ) -> (
    then:Exec
    AsyncTaskProxy:Async Action Start Listening To Stateful Event Object Reference
    OnEventReceived:Exec
    OnSavedEventStateExists:Exec
    OnEventStateCleared:Exec
    Payload:Pilgrim Event MBI Started Structure
   ) {is=title K2Node_StartListeningToStatefulEventAsyncAction}
   > Set PilgrimMBIStartedEventHandle(PilgrimMBIStartedEventHandle=@n13)
   n14 = Cast To MusicBattleInstance!(
    execute=[@n13.OnEventReceived, @n13.OnSavedEventStateExists]
    Object=Break Pilgrim Event MBI Event Base!(PilgrimEvent_MBI_EventBase:"Pilgrim Event MBI Event Base Structure (by ref)"=@n13.Payload) -> (MusicBattleInstance:Pilgrim Music Battle Instance Weak object) {is=title K2Node_BreakStruct}
   ) -> (then, CastFailed, AsMusic Battle Instance:Music Battle Instance Object Reference) {is=title K2Node_DynamicCast}
   > n15 = Cast To FortPlayerStateAthena!(execute, Object=Get PlayerState!(self:Controller Object Reference=GetOwningPlayer()) -> (PlayerState:Player State Object Reference)) -> (then, CastFailed, AsFort Player State Athena:Fort Player State Athena Object Reference) {is=title K2Node_DynamicCast}
   > Branch(Condition=IsPlayerInBattleParticipants!(self:Pilgrim Music Battle Instance Object Reference=@n14, PlayerState:Fort Player State Athena Object Reference=@n15) -> (ReturnValue:bool))
   > Set MusicBattleInstance(MusicBattleInstance:Music Battle Instance Object Reference=@n14) -> (Output_Get:Music Battle Instance Object Reference)
   Event On Deactivated()
   > SetActiveWidgetIndex(self=Get ScreenSwitcher(), Index=0)
  ]}
 Exit to Backstage {nodes [
   Exit to Backstage() {is=title K2Node_FunctionEntry}
   n1 = GetComponentByClass!(
    self:Actor Object Reference=GetOwningPlayer()
    ComponentClass:Actor Component Class Reference=/PilgrimCore/Player/Components/BP_PilgrimPlayerControllerComponent.BP_PilgrimPlayerControllerComponent_C
   ) -> (ReturnValue:BP Pilgrim Player Controller Component Object Reference)
   > Is Valid(InputObject=@n1)
   > SendRequestToStopAndSkipPostGame(self:Pilgrim Player Controller Component Object Reference=@n1)
  ]}
 Restart Song {nodes [
   Restart Song() {is=title K2Node_FunctionEntry}
   n1 = GetComponentByClass(
    self=GetOwningPlayer()
    ComponentClass=&BP_PilgrimPlayerControllerComponent_C
   )
   > Is Valid(InputObject=@n1)
   > SendRequestToRestartSong(self:Pilgrim Player Controller Component Object Reference=@n1)
  ]}
 Update All Button Status {nodes [
   Update All Button Status() {is=title K2Node_FunctionEntry}
   n1 = GetSparksMusicPlayspaceOfClass(PlayspaceType=/Script/PilgrimQuickplayRuntime.PilgrimQuickplayPlayspace, Actor=GetOwningPlayer(), bMatchChildTypes=false) -> (ReturnValue:Pilgrim Quickplay Playspace Object Reference)
   > Is Valid(InputObject=@n1)
   n2 = GetAllLocalPlayerControllers!(self:Pilgrim Playspace Object Reference=@n1) -> (ReturnValue:Array of Fort Player Controller Object References)
   n3 = Get All Player States!(self:Sparks Music Playspace Object Reference=@n1) -> (ReturnValue:Array of Player State Object References) {is=title K2Node_CallFunction}
   > n4 = Branch(Condition="Equal (Integer)"(
     A=Length(TargetArray:Array of Player State Object References=@n3)
     B=Length(TargetArray:Array of Fort Player Controller Object References=@n2)
    ))
   > SetActiveWidgetIndex(
    self:Common Visibility Switcher Object Reference=Get SoloPlay_Options() -> (SoloPlay_Options:Common Visibility Switcher Object Reference)
    Index=0
   )
   > Update Restart Button Status()
   > Update Exit Button Status()
   SetActiveWidgetIndex(execute=@n4.else, self=Get SoloPlay_Options(), Index=1)
  ]}
 Update Exit Button Status {nodes [
   Update Exit Button Status() {is=title K2Node_FunctionEntry}
   n1 = Get MusicBattleInstance() -> (MusicBattleInstance:Music Battle Instance Object Reference)
   > n2 = Is Valid(InputObject=@n1)
   SetIsEnabled(execute=@n2.Is Not Valid, self=Get ExitToBackstage_Mojo(), bInIsEnabled=false)
   SetIsEnabled(
    execute=@n2.Is Valid
    self=Get ExitToBackstage_Mojo()
    bInIsEnabled=IsHardStopAllowed!(self:Pilgrim Music Battle Instance Object Reference=@n1) -> (ReturnValue:bool)
   )
  ]}
 Update Restart Button Status {nodes [
   Update Restart Button Status() {is=title K2Node_FunctionEntry}
   n1 = Get MusicBattleInstance()
   > n2 = Is Valid(InputObject=@n1)
   SetVisibility(
    execute=[@n3.else, @n2.Is Not Valid]
    self=Get RestartButton_Mojo() -> (RestartButton_Mojo:WBP FM Core Button Regular Object Reference)
    InVisibility=Collapsed
   )
   > SetIsEnabled(self=Get RestartButton_Mojo(), bInIsEnabled=false)
   n3 = Branch(execute=@n2.Is Valid, Condition=IsRestartAllowed!(self:Pilgrim Music Battle Instance Object Reference=@n1) -> (ReturnValue:bool))
   > SetVisibility(self=Get RestartButton_Mojo(), InVisibility=SelfHitTestInvisible)
   > SetIsEnabled(self=Get RestartButton_Mojo(), bInIsEnabled=true)
  ]}
]
variables [
 Active Index {type=integer}
 MusicBattleInstance {type=MusicBattleInstance_C, default=null}
 PilgrimMBIStartedEventHandle {type=AsyncAction_StartListeningToStatefulEvent, default=null}
 SparksMusicPlayspace {type=SparksMusicPlayspace, default=null}
]
widgets {
 info {inheritedWidgetCount=0, parentClass=/Script/FortniteUI.AthenaMapScreenTabBase, rootWidgetClass=&Overlay, widgetCount=47}
 widgets [
  Overlay_Default {widgetClassPath=&Overlay, bIsVariable=true, widget=Overlay_Default}
  BackgroundBlur_93 {
   widgetClassPath=&BackgroundBlur
   parent=Overlay_Default
   widget=BackgroundBlur_93
   properties {
    blurStrength=4
    lowQualityFallbackBrush {
     bIsDynamicallyLoaded=false
     drawAs=Image
     imageType=NoImage
     mirroring=NoMirror
     resourceName=None
     resourceObject=null
     tiling=NoTile
     imageSize=<0 0>
     margin=<ltrb 0 0 0 0>
     outlineSettings {
      bUseBrushTransparency=false
      roundingType=HalfHeightRadius
      width=0
      color {colorUseRule=UseColor_Specified, specifiedColor=<rgba 0 0 0 0>}
      cornerRadii=<0 0 0 0>
     }
     tintColor {colorUseRule=UseColor_Specified, specifiedColor=<rgba 0.0083 0 0.05 0.5>}
     uVRegion {bIsValid=false, max=<0 0>, min=<0 0>}
    }
   }
   slot {class=OverlaySlot, properties {horizontalAlignment=HAlign_Fill, verticalAlignment=VAlign_Fill}}
  }
  BackgroundColor {
   widgetClassPath=&Border
   bIsVariable=true
   parent=Overlay_Default
   widget=BackgroundColor
   properties {
    renderOpacity=0.5
    background {
     bIsDynamicallyLoaded=false
     drawAs=Image
     imageType=NoImage
     mirroring=NoMirror
     resourceName=None
     resourceObject=/SparksUICommon/Materials/MI_UI_Sparks_Fullscreen_Background
     tiling=NoTile
     imageSize=<32 32>
     margin=<ltrb 0 0 0 0>
     outlineSettings {
      bUseBrushTransparency=false
      roundingType=HalfHeightRadius
      width=0
      color {colorUseRule=UseColor_Specified, specifiedColor=<rgba 0 0 0 0>}
      cornerRadii=<0 0 0 0>
     }
     tintColor {colorUseRule=UseColor_Specified, specifiedColor=<rgba 1 1 1 1>}
     uVRegion {bIsValid=false, max=<0 0>, min=<0 0>}
    }
    padding=<ltrb 0 0 0 0>
   }
   slot {class=OverlaySlot, properties {horizontalAlignment=HAlign_Fill, verticalAlignment=VAlign_Fill}}
  }
  WBP_CaptureForPostBufferUpdate {
   widgetClassPath=&WBP_CaptureForPostBufferUpdate
   bIsVariable=true
   parent=Overlay_Default
   slot {class=OverlaySlot, properties {horizontalAlignment=HAlign_Fill, verticalAlignment=VAlign_Fill}}
  }
  SafeZone_BackButton {
   widgetClassPath=&SafeZone
   bIsVariable=true
   parent=Overlay_Default
   widget=SafeZone_BackButton
   properties {padLeft=false, padTop=false}
   slot {class=OverlaySlot, properties {horizontalAlignment=HAlign_Right, verticalAlignment=VAlign_Bottom, padding=<ltrb 48 48 48 48>}}
  }
  BackButton {
   widgetClassPath=/FortUILibrary/UIKit/Buttons/WBP_UIKit_Button_Quiet
   bIsVariable=true
   parent=SafeZone_BackButton
   widget=BackButton
   properties {text=Back, triggeringInputAction {dataTable=&InputActionDataTable, rowName=GenericBack}}
   slot {class=SafeZoneSlot}
  }
  SoloPlay_Options {
   widgetClassPath=&CommonVisibilitySwitcher
   bIsVariable=true
   parent=Overlay_Default
   widget=SoloPlay_Options
   slot {class=OverlaySlot, properties {horizontalAlignment=HAlign_Center, verticalAlignment=VAlign_Center}}
  }
  ScreenSwitcher {
   widgetClassPath=&CommonActivatableWidgetSwitcher
   bIsVariable=true
   parent=SoloPlay_Options
   widget=ScreenSwitcher
   slot {class=CommonVisibilitySwitcherSlot}
  }
  SongOptionButtons {widgetClassPath=&VerticalBox, parent=ScreenSwitcher, widget=SongOptionButtons, slot {class=WidgetSwitcherSlot}}
  RestartButton_Mojo {
   widgetClassPath=/SparksUICommon/UILibrary/WBP_FM_Core_Button_Regular
   bIsVariable=true
   parent=SongOptionButtons
   widget=RestartButton_Mojo
   properties {minHeight=56, minWidth=500, text=RESTART SONG}
   slot {class=VerticalBoxSlot, properties {padding=<ltrb 10 10 10 10>}}
  }
  ExitToBackstage_Mojo {
   widgetClassPath=&WBP_FM_Core_Button_Regular
   bIsVariable=true
   parent=SongOptionButtons
   widget=ExitToBackstage_Mojo
   properties {minHeight=56, minWidth=500, text=EXIT TO BACKSTAGE}
   slot {class=VerticalBoxSlot, properties {padding=<ltrb 10 10 10 10>}}
  }
  ConfirmRestartButtons {widgetClassPath=&VerticalBox, parent=ScreenSwitcher, widget=ConfirmRestartButtons, slot {class=WidgetSwitcherSlot}}
  Overlay_0 {widgetClassPath=&Overlay, parent=ConfirmRestartButtons, widget=Overlay_0, slot {class=VerticalBoxSlot}}
  WBP_Sparks_BackingPanel {
   widgetClassPath=/PilgrimCore/Widgets/SparksUILibrary/WBP_Sparks_BackingPanel
   bIsVariable=true
   parent=Overlay_0
   widget=WBP_Sparks_BackingPanel
   properties {cornerRadius=Small, darkenLayer=false, eSparksBackingPanelType=Simple Color}
   slot {class=OverlaySlot, properties {horizontalAlignment=HAlign_Fill, verticalAlignment=VAlign_Fill}}
  }
  VerticalBox_0 {
   widgetClassPath=&VerticalBox
   parent=Overlay_0
   widget=VerticalBox_0
   properties {flowDirectionPreference=Culture}
   slot {class=OverlaySlot, properties {padding=<ltrb 64 0 64 0>}}
  }
  Spacer_159 {widgetClassPath=&Spacer, parent=VerticalBox_0, widget=Spacer_159, properties {size=<64 48>}, slot {class=VerticalBoxSlot}}
  Title_RestartSong {
   widgetClassPath=&CommonRichTextBlock
   parent=VerticalBox_0
   widget=Title_RestartSong
   properties {
    bOverrideDefaultStyle=true
    justification=Center
    text=RESTART SONG
    defaultTextStyleOverride {
     overflowPolicy=Clip
     transformPolicy=None
     colorAndOpacity {colorUseRule=UseColor_Specified, specifiedColor=<rgba 1 1 1 1>}
     font {
      bForceMonospaced=false
      bMaterialIsStencil=false
      fontMaterial=null
      fontObject=&HeadingNow
      letterSpacing=0
      monospacedWidth=1
      size=25.5
      skewAmount=0
      typefaceFontName=86 Bold Italic
      outlineSettings {
       bApplyOutlineToDropShadows=false
       bMiteredCorners=false
       bSeparateFillAlpha=false
       outlineMaterial=null
       outlineSize=0
       outlineColor=<rgba 0 0 0 1>
      }
     }
     highlightColor {colorUseRule=UseColor_Specified, specifiedColor=<rgba 0 0 0 1>}
     highlightShape {
      bIsDynamicallyLoaded=false
      drawAs=Image
      imageType=NoImage
      mirroring=NoMirror
      resourceName=None
      resourceObject=null
      tiling=NoTile
      imageSize=<32 32>
      margin=<ltrb 0 0 0 0>
      outlineSettings {
       bUseBrushTransparency=false
       roundingType=HalfHeightRadius
       width=0
       color {colorUseRule=UseColor_Specified, specifiedColor=<rgba 0 0 0 0>}
       cornerRadii=<0 0 0 0>
      }
      tintColor {colorUseRule=UseColor_Specified, specifiedColor=<rgba 1 1 1 1>}
      uVRegion {bIsValid=false, max=<0 0>, min=<0 0>}
     }
     selectedBackgroundColor {colorUseRule=UseColor_Foreground, specifiedColor=<rgba 1 0 1 1>}
     shadowColorAndOpacity=<rgba 0 0 0 1>
     shadowOffset=<0 0>
     strikeBrush {
      bIsDynamicallyLoaded=false
      drawAs=Image
      imageType=NoImage
      mirroring=NoMirror
      resourceName=None
      resourceObject=null
      tiling=NoTile
      imageSize=<32 32>
      margin=<ltrb 0 0 0 0>
      outlineSettings {
       bUseBrushTransparency=false
       roundingType=HalfHeightRadius
       width=0
       color {colorUseRule=UseColor_Specified, specifiedColor=<rgba 0 0 0 0>}
       cornerRadii=<0 0 0 0>
      }
      tintColor {colorUseRule=UseColor_Specified, specifiedColor=<rgba 1 1 1 1>}
      uVRegion {bIsValid=false, max=<0 0>, min=<0 0>}
     }
     underlineBrush {
      bIsDynamicallyLoaded=false
      drawAs=Image
      imageType=NoImage
      mirroring=NoMirror
      resourceName=None
      resourceObject=null
      tiling=NoTile
      imageSize=<32 32>
      margin=<ltrb 0 0 0 0>
      outlineSettings {
       bUseBrushTransparency=false
       roundingType=HalfHeightRadius
       width=0
       color {colorUseRule=UseColor_Specified, specifiedColor=<rgba 0 0 0 0>}
       cornerRadii=<0 0 0 0>
      }
      tintColor {colorUseRule=UseColor_Specified, specifiedColor=<rgba 1 1 1 1>}
      uVRegion {bIsValid=false, max=<0 0>, min=<0 0>}
     }
    }
   }
   slot {class=VerticalBoxSlot, properties {horizontalAlignment=HAlign_Left, verticalAlignment=VAlign_Top}}
  }
  Spacer_1 {widgetClassPath=&Spacer, parent=VerticalBox_0, widget=Spacer_1, properties {size=<80 20>}, slot {class=VerticalBoxSlot}}
  Description_Restart {
   widgetClassPath=&CommonRichTextBlock
   parent=VerticalBox_0
   widget=Description_Restart
   properties {
    bOverrideDefaultStyle=true
    justification=Center
    text=Do you really want to restart the song?
    wrapTextAt=750
    defaultTextStyleOverride {
     overflowPolicy=Clip
     transformPolicy=None
     colorAndOpacity {colorUseRule=UseColor_Specified, specifiedColor=<rgba 0.5 0.8333 1 1>}
     font {
      bForceMonospaced=false
      bMaterialIsStencil=false
      fontMaterial=null
      fontObject=&HeadingNow
      letterSpacing=0
      monospacedWidth=1
      size=15.75
      skewAmount=0
      typefaceFontName=74 Regular
      outlineSettings {
       bApplyOutlineToDropShadows=false
       bMiteredCorners=false
       bSeparateFillAlpha=false
       outlineMaterial=null
       outlineSize=0
       outlineColor=<rgba 0 0 0 1>
      }
     }
     highlightColor {colorUseRule=UseColor_Specified, specifiedColor=<rgba 0 0 0 1>}
     highlightShape {
      bIsDynamicallyLoaded=false
      drawAs=Image
      imageType=NoImage
      mirroring=NoMirror
      resourceName=None
      resourceObject=null
      tiling=NoTile
      imageSize=<32 32>
      margin=<ltrb 0 0 0 0>
      outlineSettings {
       bUseBrushTransparency=false
       roundingType=HalfHeightRadius
       width=0
       color {colorUseRule=UseColor_Specified, specifiedColor=<rgba 0 0 0 0>}
       cornerRadii=<0 0 0 0>
      }
      tintColor {colorUseRule=UseColor_Specified, specifiedColor=<rgba 1 1 1 1>}
      uVRegion {bIsValid=false, max=<0 0>, min=<0 0>}
     }
     selectedBackgroundColor {colorUseRule=UseColor_Foreground, specifiedColor=<rgba 1 0 1 1>}
     shadowColorAndOpacity=<rgba 0 0 0 1>
     shadowOffset=<0 0>
     strikeBrush {
      bIsDynamicallyLoaded=false
      drawAs=Image
      imageType=NoImage
      mirroring=NoMirror
      resourceName=None
      resourceObject=null
      tiling=NoTile
      imageSize=<32 32>
      margin=<ltrb 0 0 0 0>
      outlineSettings {
       bUseBrushTransparency=false
       roundingType=HalfHeightRadius
       width=0
       color {colorUseRule=UseColor_Specified, specifiedColor=<rgba 0 0 0 0>}
       cornerRadii=<0 0 0 0>
      }
      tintColor {colorUseRule=UseColor_Specified, specifiedColor=<rgba 1 1 1 1>}
      uVRegion {bIsValid=false, max=<0 0>, min=<0 0>}
     }
     underlineBrush {
      bIsDynamicallyLoaded=false
      drawAs=Image
      imageType=NoImage
      mirroring=NoMirror
      resourceName=None
      resourceObject=null
      tiling=NoTile
      imageSize=<32 32>
      margin=<ltrb 0 0 0 0>
      outlineSettings {
       bUseBrushTransparency=false
       roundingType=HalfHeightRadius
       width=0
       color {colorUseRule=UseColor_Specified, specifiedColor=<rgba 0 0 0 0>}
       cornerRadii=<0 0 0 0>
      }
      tintColor {colorUseRule=UseColor_Specified, specifiedColor=<rgba 1 1 1 1>}
      uVRegion {bIsValid=false, max=<0 0>, min=<0 0>}
     }
    }
   }
   slot {class=VerticalBoxSlot, properties {horizontalAlignment=HAlign_Left, verticalAlignment=VAlign_Top}}
  }
  Spacer_2 {widgetClassPath=&Spacer, parent=VerticalBox_0, widget=Spacer_2, properties {size=<750 120>}, slot {class=VerticalBoxSlot}}
  HBOX_Buttons {
   widgetClassPath=&HorizontalBox
   parent=VerticalBox_0
   widget=HBOX_Buttons
   slot {class=VerticalBoxSlot, properties {horizontalAlignment=HAlign_Center, verticalAlignment=VAlign_Center, size {sizeRule=Fill, value=1}}}
  }
  YesRestartButton_Mojo {
   widgetClassPath=&WBP_FM_Core_Button_Regular
   bIsVariable=true
   parent=HBOX_Buttons
   widget=YesRestartButton_Mojo
   properties {minHeight=56, minWidth=364, text=YES}
   slot {class=HorizontalBoxSlot, properties {padding=<ltrb 10 10 10 10>}}
  }
  Spacer_4 {widgetClassPath=&Spacer, parent=HBOX_Buttons, widget=Spacer_4, properties {size=<42 1>}, slot {class=HorizontalBoxSlot}}
  NoRestartButton_Mojo {
   widgetClassPath=&WBP_FM_Core_Button_Regular
   bIsVariable=true
   parent=HBOX_Buttons
   widget=NoRestartButton_Mojo
   properties {minHeight=56, minWidth=364, text=NO}
   slot {class=HorizontalBoxSlot, properties {padding=<ltrb 10 10 10 10>}}
  }
  Spacer_3 {widgetClassPath=&Spacer, parent=VerticalBox_0, widget=Spacer_3, properties {size=<80 48>}, slot {class=VerticalBoxSlot}}
  ConfirmExitButtons {
   widgetClassPath=&VerticalBox
   parent=ScreenSwitcher
   widget=ConfirmExitButtons
   properties {bOverride_Cursor=true}
   slot {class=WidgetSwitcherSlot}
  }
  Overlay_1 {widgetClassPath=&Overlay, parent=ConfirmExitButtons, widget=Overlay_1, slot {class=VerticalBoxSlot}}
  WBP_Sparks_BackingPanel_123 {
   widgetClassPath=&WBP_Sparks_BackingPanel
   bIsVariable=true
   parent=Overlay_1
   widget=WBP_Sparks_BackingPanel_123
   properties {cornerRadius=Small, darkenLayer=false, eSparksBackingPanelType=Simple Color}
   slot {class=OverlaySlot, properties {horizontalAlignment=HAlign_Fill, verticalAlignment=VAlign_Fill}}
  }
  VerticalBox {
   widgetClassPath=&VerticalBox
   parent=Overlay_1
   widget=VerticalBox
   properties {flowDirectionPreference=Culture}
   slot {class=OverlaySlot, properties {padding=<ltrb 64 0 64 0>}}
  }
  Spacer {widgetClassPath=&Spacer, parent=VerticalBox, widget=Spacer, properties {size=<80 48>}, slot {class=VerticalBoxSlot}}
  Title_Exit {
   widgetClassPath=&CommonRichTextBlock
   parent=VerticalBox
   widget=Title_Exit
   properties {
    bOverrideDefaultStyle=true
    justification=Center
    text=EXIT?
    defaultTextStyleOverride {
     overflowPolicy=Clip
     transformPolicy=None
     colorAndOpacity {colorUseRule=UseColor_Specified, specifiedColor=<rgba 1 1 1 1>}
     font {
      bForceMonospaced=false
      bMaterialIsStencil=false
      fontMaterial=null
      fontObject=&HeadingNow
      letterSpacing=0
      monospacedWidth=1
      size=25.5
      skewAmount=0
      typefaceFontName=86 Bold Italic
      outlineSettings {
       bApplyOutlineToDropShadows=false
       bMiteredCorners=false
       bSeparateFillAlpha=false
       outlineMaterial=null
       outlineSize=0
       outlineColor=<rgba 0 0 0 1>
      }
     }
     highlightColor {colorUseRule=UseColor_Specified, specifiedColor=<rgba 0 0 0 1>}
     highlightShape {
      bIsDynamicallyLoaded=false
      drawAs=Image
      imageType=NoImage
      mirroring=NoMirror
      resourceName=None
      resourceObject=null
      tiling=NoTile
      imageSize=<32 32>
      margin=<ltrb 0 0 0 0>
      outlineSettings {
       bUseBrushTransparency=false
       roundingType=HalfHeightRadius
       width=0
       color {colorUseRule=UseColor_Specified, specifiedColor=<rgba 0 0 0 0>}
       cornerRadii=<0 0 0 0>
      }
      tintColor {colorUseRule=UseColor_Specified, specifiedColor=<rgba 1 1 1 1>}
      uVRegion {bIsValid=false, max=<0 0>, min=<0 0>}
     }
     selectedBackgroundColor {colorUseRule=UseColor_Foreground, specifiedColor=<rgba 1 0 1 1>}
     shadowColorAndOpacity=<rgba 0 0 0 1>
     shadowOffset=<0 0>
     strikeBrush {
      bIsDynamicallyLoaded=false
      drawAs=Image
      imageType=NoImage
      mirroring=NoMirror
      resourceName=None
      resourceObject=null
      tiling=NoTile
      imageSize=<32 32>
      margin=<ltrb 0 0 0 0>
      outlineSettings {
       bUseBrushTransparency=false
       roundingType=HalfHeightRadius
       width=0
       color {colorUseRule=UseColor_Specified, specifiedColor=<rgba 0 0 0 0>}
       cornerRadii=<0 0 0 0>
      }
      tintColor {colorUseRule=UseColor_Specified, specifiedColor=<rgba 1 1 1 1>}
      uVRegion {bIsValid=false, max=<0 0>, min=<0 0>}
     }
     underlineBrush {
      bIsDynamicallyLoaded=false
      drawAs=Image
      imageType=NoImage
      mirroring=NoMirror
      resourceName=None
      resourceObject=null
      tiling=NoTile
      imageSize=<32 32>
      margin=<ltrb 0 0 0 0>
      outlineSettings {
       bUseBrushTransparency=false
       roundingType=HalfHeightRadius
       width=0
       color {colorUseRule=UseColor_Specified, specifiedColor=<rgba 0 0 0 0>}
       cornerRadii=<0 0 0 0>
      }
      tintColor {colorUseRule=UseColor_Specified, specifiedColor=<rgba 1 1 1 1>}
      uVRegion {bIsValid=false, max=<0 0>, min=<0 0>}
     }
    }
   }
   slot {class=VerticalBoxSlot, properties {horizontalAlignment=HAlign_Left, verticalAlignment=VAlign_Top}}
  }
  Spacer_5 {widgetClassPath=&Spacer, parent=VerticalBox, widget=Spacer_5, properties {size=<80 20>}, slot {class=VerticalBoxSlot}}
  Description_Exit {
   widgetClassPath=&CommonRichTextBlock
   parent=VerticalBox
   widget=Description_Exit
   properties {
    bOverrideDefaultStyle=true
    justification=Center
    text=Leave the match and exit to backstage?
    wrapTextAt=750
    defaultTextStyleOverride {
     overflowPolicy=Clip
     transformPolicy=None
     colorAndOpacity {colorUseRule=UseColor_Specified, specifiedColor=<rgba 0.5 0.8333 1 1>}
     font {
      bForceMonospaced=false
      bMaterialIsStencil=false
      fontMaterial=null
      fontObject=&HeadingNow
      letterSpacing=0
      monospacedWidth=1
      size=15.75
      skewAmount=0
      typefaceFontName=74 Regular
      outlineSettings {
       bApplyOutlineToDropShadows=false
       bMiteredCorners=false
       bSeparateFillAlpha=false
       outlineMaterial=null
       outlineSize=0
       outlineColor=<rgba 0 0 0 1>
      }
     }
     highlightColor {colorUseRule=UseColor_Specified, specifiedColor=<rgba 0 0 0 1>}
     highlightShape {
      bIsDynamicallyLoaded=false
      drawAs=Image
      imageType=NoImage
      mirroring=NoMirror
      resourceName=None
      resourceObject=null
      tiling=NoTile
      imageSize=<32 32>
      margin=<ltrb 0 0 0 0>
      outlineSettings {
       bUseBrushTransparency=false
       roundingType=HalfHeightRadius
       width=0
       color {colorUseRule=UseColor_Specified, specifiedColor=<rgba 0 0 0 0>}
       cornerRadii=<0 0 0 0>
      }
      tintColor {colorUseRule=UseColor_Specified, specifiedColor=<rgba 1 1 1 1>}
      uVRegion {bIsValid=false, max=<0 0>, min=<0 0>}
     }
     selectedBackgroundColor {colorUseRule=UseColor_Foreground, specifiedColor=<rgba 1 0 1 1>}
     shadowColorAndOpacity=<rgba 0 0 0 1>
     shadowOffset=<0 0>
     strikeBrush {
      bIsDynamicallyLoaded=false
      drawAs=Image
      imageType=NoImage
      mirroring=NoMirror
      resourceName=None
      resourceObject=null
      tiling=NoTile
      imageSize=<32 32>
      margin=<ltrb 0 0 0 0>
      outlineSettings {
       bUseBrushTransparency=false
       roundingType=HalfHeightRadius
       width=0
       color {colorUseRule=UseColor_Specified, specifiedColor=<rgba 0 0 0 0>}
       cornerRadii=<0 0 0 0>
      }
      tintColor {colorUseRule=UseColor_Specified, specifiedColor=<rgba 1 1 1 1>}
      uVRegion {bIsValid=false, max=<0 0>, min=<0 0>}
     }
     underlineBrush {
      bIsDynamicallyLoaded=false
      drawAs=Image
      imageType=NoImage
      mirroring=NoMirror
      resourceName=None
      resourceObject=null
      tiling=NoTile
      imageSize=<32 32>
      margin=<ltrb 0 0 0 0>
      outlineSettings {
       bUseBrushTransparency=false
       roundingType=HalfHeightRadius
       width=0
       color {colorUseRule=UseColor_Specified, specifiedColor=<rgba 0 0 0 0>}
       cornerRadii=<0 0 0 0>
      }
      tintColor {colorUseRule=UseColor_Specified, specifiedColor=<rgba 1 1 1 1>}
      uVRegion {bIsValid=false, max=<0 0>, min=<0 0>}
     }
    }
   }
   slot {class=VerticalBoxSlot, properties {horizontalAlignment=HAlign_Left, verticalAlignment=VAlign_Top}}
  }
  Spacer_6 {widgetClassPath=&Spacer, parent=VerticalBox, widget=Spacer_6, properties {size=<750 120>}, slot {class=VerticalBoxSlot}}
  HBOX_Buttons_1 {
   widgetClassPath=&HorizontalBox
   parent=VerticalBox
   widget=HBOX_Buttons_1
   slot {class=VerticalBoxSlot, properties {horizontalAlignment=HAlign_Center, verticalAlignment=VAlign_Center, size {sizeRule=Fill, value=1}}}
  }
  YesLeaveButton_Mojo {
   widgetClassPath=&WBP_FM_Core_Button_Regular
   bIsVariable=true
   parent=HBOX_Buttons_1
   widget=YesLeaveButton_Mojo
   properties {minHeight=56, minWidth=364, text=YES}
   slot {class=HorizontalBoxSlot, properties {padding=<ltrb 10 10 10 10>}}
  }
  Spacer_7 {widgetClassPath=&Spacer, parent=HBOX_Buttons_1, widget=Spacer_7, properties {size=<42 1>}, slot {class=HorizontalBoxSlot}}
  NoReturnButton_Mojo {
   widgetClassPath=&WBP_FM_Core_Button_Regular
   bIsVariable=true
   parent=HBOX_Buttons_1
   widget=NoReturnButton_Mojo
   properties {minHeight=56, minWidth=364, text=NO}
   slot {class=HorizontalBoxSlot, properties {padding=<ltrb 10 10 10 10>}}
  }
  Spacer_8 {widgetClassPath=&Spacer, parent=VerticalBox, widget=Spacer_8, properties {size=<80 48>}, slot {class=VerticalBoxSlot}}
  GridPanel_73 {
   widgetClassPath=&GridPanel
   parent=SoloPlay_Options
   widget=GridPanel_73
   properties {columnFill [1], rowFill [1]}
   slot {class=CommonVisibilitySwitcherSlot, properties {horizontalAlignment=HAlign_Center, verticalAlignment=VAlign_Center}}
  }
  SoloPlay_Options_BG {
   widgetClassPath=&Image
   parent=GridPanel_73
   widget=SoloPlay_Options_BG
   properties {colorAndOpacity=<rgba 0.0356 0.0015 0.0003 1>}
   slot {class=GridSlot}
  }
  GridPanel_75 {
   widgetClassPath=&GridPanel
   parent=GridPanel_73
   widget=GridPanel_75
   slot {class=GridSlot, properties {horizontalAlignment=HAlign_Left, verticalAlignment=VAlign_Center, padding=<ltrb 20 20 20 20>}}
  }
  Image_Exclamation {
   widgetClassPath=&Image
   parent=GridPanel_75
   widget=Image_Exclamation
   properties {brush {
     bIsDynamicallyLoaded=false
     drawAs=Image
     imageType=NoImage
     mirroring=NoMirror
     resourceName=None
     resourceObject=/Game/UI/ContentControls/Textures/T_UI_Icon_Warning
     tiling=NoTile
     imageSize=<32 32>
     margin=<ltrb 0 0 0 0>
     outlineSettings {
      bUseBrushTransparency=false
      roundingType=HalfHeightRadius
      width=0
      color {colorUseRule=UseColor_Specified, specifiedColor=<rgba 0 0 0 0>}
      cornerRadii=<0 0 0 0>
     }
     tintColor {colorUseRule=UseColor_Specified, specifiedColor=<rgba 0.8879 0.147 0.147 1>}
     uVRegion {bIsValid=false, max=<0 0>, min=<0 0>}
    }}
   slot {class=GridSlot, properties {horizontalAlignment=HAlign_Left, verticalAlignment=VAlign_Center}}
  }
  SoloPlay_Options_Text {
   widgetClassPath=&CommonRichTextBlock
   parent=GridPanel_75
   widget=SoloPlay_Options_Text
   properties {
    bOverrideDefaultStyle=true
    defaultTextStyleOverrideClass=/FortUILibrary/Foundation/Buttons/ButtonStyles/Athena/TextStyle_BurbSmBk_15_Red.TextStyle_BurbSmBk_15_Red_C
    text="Song options only available in\r\nsolo player mode."
    defaultTextStyleOverride {
     overflowPolicy=Clip
     transformPolicy=None
     colorAndOpacity {colorUseRule=UseColor_Specified, specifiedColor=<rgba 1 1 1 1>}
     font {
      bForceMonospaced=false
      bMaterialIsStencil=false
      fontMaterial=null
      fontObject=&NotoSans
      letterSpacing=0
      monospacedWidth=1
      size=24
      skewAmount=0
      typefaceFontName=NotoSans-Regular
      outlineSettings {
       bApplyOutlineToDropShadows=false
       bMiteredCorners=false
       bSeparateFillAlpha=false
       outlineMaterial=null
       outlineSize=0
       outlineColor=<rgba 0 0 0 1>
      }
     }
     highlightColor {colorUseRule=UseColor_Specified, specifiedColor=<rgba 0 0 0 1>}
     highlightShape {
      bIsDynamicallyLoaded=false
      drawAs=Image
      imageType=NoImage
      mirroring=NoMirror
      resourceName=None
      resourceObject=null
      tiling=NoTile
      imageSize=<32 32>
      margin=<ltrb 0 0 0 0>
      outlineSettings {
       bUseBrushTransparency=false
       roundingType=HalfHeightRadius
       width=0
       color {colorUseRule=UseColor_Specified, specifiedColor=<rgba 0 0 0 0>}
       cornerRadii=<0 0 0 0>
      }
      tintColor {colorUseRule=UseColor_Specified, specifiedColor=<rgba 1 1 1 1>}
      uVRegion {bIsValid=false, max=<0 0>, min=<0 0>}
     }
     selectedBackgroundColor {colorUseRule=UseColor_Foreground, specifiedColor=<rgba 1 0 1 1>}
     shadowColorAndOpacity=<rgba 0 0 0 1>
     shadowOffset=<0 0>
     strikeBrush {
      bIsDynamicallyLoaded=false
      drawAs=Image
      imageType=NoImage
      mirroring=NoMirror
      resourceName=None
      resourceObject=null
      tiling=NoTile
      imageSize=<32 32>
      margin=<ltrb 0 0 0 0>
      outlineSettings {
       bUseBrushTransparency=false
       roundingType=HalfHeightRadius
       width=0
       color {colorUseRule=UseColor_Specified, specifiedColor=<rgba 0 0 0 0>}
       cornerRadii=<0 0 0 0>
      }
      tintColor {colorUseRule=UseColor_Specified, specifiedColor=<rgba 1 1 1 1>}
      uVRegion {bIsValid=false, max=<0 0>, min=<0 0>}
     }
     underlineBrush {
      bIsDynamicallyLoaded=false
      drawAs=Image
      imageType=NoImage
      mirroring=NoMirror
      resourceName=None
      resourceObject=null
      tiling=NoTile
      imageSize=<32 32>
      margin=<ltrb 0 0 0 0>
      outlineSettings {
       bUseBrushTransparency=false
       roundingType=HalfHeightRadius
       width=0
       color {colorUseRule=UseColor_Specified, specifiedColor=<rgba 0 0 0 0>}
       cornerRadii=<0 0 0 0>
      }
      tintColor {colorUseRule=UseColor_Specified, specifiedColor=<rgba 1 1 1 1>}
      uVRegion {bIsValid=false, max=<0 0>, min=<0 0>}
     }
    }
   }
   slot {class=GridSlot, properties {horizontalAlignment=HAlign_Left, verticalAlignment=VAlign_Center, padding=<ltrb 38 0 0 0>}}
  }
  SafeZone_PInverse {
   widgetClassPath=&SafeZone
   bIsVariable=true
   parent=Overlay_Default
   widget=SafeZone_PInverse
   properties {padLeft=false, padRight=false, visibility=Collapsed}
   slot {class=OverlaySlot, properties {horizontalAlignment=HAlign_Fill, verticalAlignment=VAlign_Fill, padding=<ltrb 250 0 250 0>}}
  }
  ScaleBox_PInverse {
   widgetClassPath=&ScaleBox
   bIsVariable=true
   parent=SafeZone_PInverse
   widget=ScaleBox_PInverse
   properties {stretch=ScaleToFitX, stretchDirection=UpOnly}
   slot {class=SafeZoneSlot}
  }
  Overlay_PInverse {
   widgetClassPath=&Overlay
   bIsVariable=true
   parent=ScaleBox_PInverse
   widget=Overlay_PInverse
   slot {class=ScaleBoxSlot, properties {horizontalAlignment=HAlign_Fill, verticalAlignment=VAlign_Fill}}
  }
 ]
}
