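# Gameplay quest panel widget for Pilgrim Quickplay, built on the native SparksQuestScreenContainer class.
# A full-screen canvas hosts the inherited AthenaScreenContainer slot, filled by WBP_PilgrimQuestScreenContainer and collapsed by default.
# That container overrides its action domain to Menu, consumes pointer input and uses the Quickplay song menu as its default tab.
# The event graph holds only empty stubs, so the Blueprint adds no logic of its own.
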WidgetBlueprint /PilgrimQuickplay/Widgets/Quests/WBP_PilgrimGameplayQuestPanel
graphs [EventGraph {nodes [
   Event PreConstruct()
   Event Construct()
   Event Tick()
  ]}]
variables []
widgets {
 info {inheritedWidgetCount=1, parentClass=/Script/SparksCoreUI.SparksQuestScreenContainer, rootWidgetClass=&CanvasPanel, widgetCount=2}
 widgets [
  CanvasPanel_19 {widgetClassPath=&CanvasPanel, widget=CanvasPanel_19}
  AthenaScreenContainer {
   widgetClassPath=&/WBP_PilgrimQuestScreenContainer
   bInherited=true
   bIsVariable=true
   parent=CanvasPanel_19
   widget=AthenaScreenContainer
   properties {
    actionDomainOverride=/Game/UI/ActionDomain/ActionDomain_Menu
    bConsumePointerInput=true
    bOverrideActionDomain=true
    visibility=Collapsed
    defaultTabs [/PilgrimQuickplay/Widgets/SongMenu/WBP_PilgrimQuickplay_SongMenu.WBP_PilgrimQuickplay_SongMenu_C]
   }
   slot {
    class=CanvasPanelSlot
    properties {bAutoSize=true, layoutData {alignment=<0 0>, anchors {maximum=<1 1>, minimum=<0 0>}, offsets=<ltrb 0 0 0 0>}}
   }
  }
 ]
}
