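# Quest reward widget for the minimap goals page, derived from AthenaChallengeRewards, showing a reward item with a cosmetic gating warning icon.
# An overlay holds the inherited item container, a dynamic entry box of multi-reward markers, and the warning image.
# When an item is presented, it casts it to a cosmetic definition and shows the icon only if rating restrictions apply to the owning local player, collapsing it otherwise.
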WidgetBlueprint /FortniteCoreUI/HUD/MiniMap/GoalsPage/GoalsRewards_Quest
graphs [
 EventGraph {nodes [
   n1 = Event BP_OnItemPresented!() -> (OutputDelegate, then, Item:Item Definition Base Object Reference) {is=title K2Node_Event}
   > UpdateCosmeticGateWarning!(execute, self, Item:Item Definition Base Object Reference=@n1.Item) -> (then)
  ]}
 UpdateCosmeticGateWarning {nodes [
   n1 = UpdateCosmeticGateWarning!() -> (then, Item:Item Definition Base Object Reference) {is=title K2Node_FunctionEntry}
   > n2 = Cast To AthenaCosmeticItemDefinition!(execute, Object=@n1) -> (then, CastFailed, AsAthena Cosmetic Item Definition:Athena Cosmetic Item Definition Object Reference) {is=title K2Node_DynamicCast}
   > n3 = HasRatingRestrictions!(
    execute
    LocalPlayer:Local Player Object Reference=GetOwningLocalPlayer!(self:Widget Object Reference) -> (ReturnValue:Local Player Object Reference)
    CosmeticItemGatingInterface:"Cosmetic Item Gating Interface Interface (by ref)"=@n2
   ) -> (then, ReturnValue:bool)
   > n4 = Branch(Condition=@n3)
   n5 = Get Image_CosmeticGate() -> (Image_CosmeticGate:Image Object Reference)
   > SetVisibility(self=@n5, InVisibility=SelfHitTestInvisible)
   SetVisibility(execute=[@n4.else, @n2.CastFailed], self=@n5, InVisibility=Collapsed)
  ]}
]
variables []
widgets {
 info {inheritedWidgetCount=2, parentClass=/Script/FortniteUI.AthenaChallengeRewards, rootWidgetClass=&Overlay, widgetCount=5}
 widgets [
  Overlay_61 {widgetClassPath=&Overlay, widget=Overlay_61}
  ItemContainer {
   widgetClassPath=&Overlay
   bInherited=true
   parent=Overlay_61
   widget=ItemContainer
   slot {class=OverlaySlot, properties {horizontalAlignment=HAlign_Center, verticalAlignment=VAlign_Fill}}
  }
  CanvasPanel_1 {
   widgetClassPath=&CanvasPanel
   parent=Overlay_61
   widget=CanvasPanel_1
   slot {class=OverlaySlot, properties {horizontalAlignment=HAlign_Center, verticalAlignment=VAlign_Bottom, padding=<ltrb 0 0 0 -6>}}
  }
  DynamicBox_MultiMarkers {
   widgetClassPath=/Script/FortniteUI.FortDynamicEntryBox
   bInherited=true
   bIsVariable=true
   parent=CanvasPanel_1
   widget=DynamicBox_MultiMarkers
   properties {
    entryHorizontalAlignment=HAlign_Center
    entryVerticalAlignment=VAlign_Center
    entryWidgetClass=/FortniteCoreUI/HUD/MiniMap/AthenaChallengeRewardsMultiMarker.AthenaChallengeRewardsMultiMarker_C
   }
   slot {
    class=CanvasPanelSlot
    properties {bAutoSize=true, layoutData {alignment=<0.5 0.5>, anchors {maximum=<0.5 0.5>, minimum=<0.5 0.5>}, offsets=<ltrb 0 0 100 30>}}
   }
  }
  Image_CosmeticGate {
   widgetClassPath=&Image
   bIsVariable=true
   parent=Overlay_61
   widget=Image_CosmeticGate
   properties {
    visibility=Collapsed
    brush {
     bIsDynamicallyLoaded=false
     drawAs=Image
     imageType=FullColor
     mirroring=NoMirror
     resourceName=None
     resourceObject=/Game/Athena/UI/Frontend/CosmeticGating/Materials/MI_UI_CosmeticGate_WarningIcon
     tiling=NoTile
     imageSize=<40 40>
     margin=<ltrb 0 0 0 0>
     outlineSettings {
      bUseBrushTransparency=false
      roundingType=HalfHeightRadius
      width=0
      color {colorUseRule=UseColor_Specified, specifiedColor=<rgba 0 0 0 0>}
      cornerRadii=<0 0 0 0>
     }
     tintColor {colorUseRule=UseColor_Specified, specifiedColor=<rgba 1 1 1 1>}
     uVRegion {bIsValid=false, max=<0 0>, min=<0 0>}
    }
   }
   slot {class=OverlaySlot, properties {horizontalAlignment=HAlign_Center, verticalAlignment=VAlign_Bottom, padding=<ltrb -35 0 0 0>}}
  }
 ]
}
